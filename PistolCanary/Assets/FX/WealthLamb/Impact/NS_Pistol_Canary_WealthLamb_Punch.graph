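# Niagara impact effect for the Pistol Canary WealthLamb punch: a layered blast of shockwave, smoke, sparks and air distortion-style meshes.
# A delayed triple-renderer sprite shockwave joins two expanding, fading spheres, a short-lived wind cone and flipbook smoke with distance-based fading.
# GPU sparks burst in three loops from randomly tilted rings, swirled by curl noise and stretched by speed.
# User.Uniform Impact Scale drives size across emitters, plus spark speed and noise strength.
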NiagaraSystem /PistolCanary/Assets/FX/WealthLamb/Impact/NS_Pistol_Canary_WealthLamb_Punch
dependencies {
 usedDataInterfaces [&NiagaraDataInterfaceCurve, &NiagaraDataInterfaceColorCurve, &NiagaraDataInterfaceArrayFloat3]
 usedDynamicInputs []
 usedModules [&SystemState]
}
emitters {
 ShockWave {
  emitterUpdateScript [
   &EmitterState(Life Cycle Mode=::NewEnumerator1, Loop Behavior=::NewEnumerator1, Loop Delay=$System.BurstTime, Loop Duration=1, UseLoopDelay=true)
   &SpawnBurst_Instantaneous(Use Spawn Probability=false, UseLoopCountLimit=false)
  ]
  particleSpawnScript [&InitializeParticle(
    Color=<rgba 0.1927 0.1927 0.1927 0.1>
    Sprite Size Mode=::NewEnumerator3
    Uniform Sprite Size=2000
    UsePositionOffset=false
    Write Material Random=true
   )]
  particleUpdateScript [
   &ParticleState()
   &ScaleSpriteSize(
    Uniform Curve Scale=$User.Uniform Impact Scale
    Uniform Curve Sprite Scale=DataInterface() {propertyValues {Curve {keys [{arriveTangent=1.5088, interpMode=RCIM_Cubic, leaveTangent=1.5088, tangentMode=RCTM_User, time=0.1, value=0.25}, <at 1 1 RCIM_Cubic>]}}}
   )
   &DynamicMaterialParameters(
    Index 0 Param 4=&FloatFromCurve(
     FloatCurve=DataInterface() {propertyValues {Curve {
        defaultValue=1
        keys [
         {arriveTangent=9.0725, interpMode=RCIM_Cubic, leaveTangent=9.0725, tangentMode=RCTM_User, time=0, value=0}
         {arriveTangent=-1.4853, interpMode=RCIM_Cubic, leaveTangent=-1.4853, tangentMode=RCTM_User, time=0.1, value=0.975}
         {arriveTangent=-0.0405, interpMode=RCIM_Cubic, leaveTangent=-0.0405, tangentMode=RCTM_User, time=1, value=0}
        ]
       }}}
     Scale Curve=-1
    )
    Module.Index0Float0Write=true
    Module.Index0Float1Write=true
    Module.Index0Float2Write=true
    Module.Index0Float3Write=true
   )
   /SetVariables(
    Particles.SpriteSize001:Vector2f=&Multiply_Vector2D_ByFloat(
     Float=&FloatFromCurve(FloatCurve=DataInterface() {propertyValues {Curve {defaultValue=1, keys [{tangentMode=RCTM_User, time=0, value=3}, {tangentMode=RCTM_User, time=1, value=2}]}}})
     Vector2D=$Particles.SpriteSize
    )
    Particles.SpriteSize002:Vector2f=&Multiply_Vector2D_ByFloat(
     Float=&FloatFromCurve(FloatCurve=DataInterface() {propertyValues {Curve {defaultValue=1, keys [{tangentMode=RCTM_User, time=0, value=0.75}, {tangentMode=RCTM_User, time=1, value=0.25}]}}})
     Vector2D=$Particles.SpriteSize
    )
   ) {isSetParameters=true}
   &ScaleColor(
    Scale Alpha=&FloatFromCurve(FloatCurve=DataInterface() {propertyValues {Curve {keys [<at 0.5 1 RCIM_Cubic>,<at 1 0 RCIM_Cubic>]}}})
    ScaleA=true
    ScaleRGB=true
   )
  ]
  rendererClasses [&NiagaraSpriteRendererProperties]
  renderers [
   {rendererClass=&NiagaraSpriteRendererProperties}
   {rendererClass=&NiagaraSpriteRendererProperties, rendererIndex=1}
   {rendererClass=&NiagaraSpriteRendererProperties, rendererIndex=2}
  ]
 }
 Smoke {
  emitterUpdateScript [
   &EmitterState(Life Cycle Mode=::NewEnumerator1, Loop Behavior=::NewEnumerator1, Loop Duration=1, UseLoopDelay=true)
   &SpawnBurst_Instantaneous(Spawn Count=8, Spawn Time=0.15, Use Spawn Probability=false, UseLoopCountLimit=true)
   &SpawnBurst_Instantaneous(Spawn Group=1, Spawn Time=0.05, Use Spawn Probability=false, UseLoopCountLimit=false) {module=SpawnBurst_Instantaneous001}
  ]
  particleSpawnScript [
   &InitializeParticle(
    Color Maximum=<rgba 0.3854 0.3854 0.3854 0.4435>
    Color Minimum=<rgba 0.2552 0.2552 0.2552 0.2957>
    Color Mode=::NewEnumerator2
    Lifetime=&MaskFloatBySpawnGroup(Masked Float=&RandomRangeFloat(Minimum=0.5, Maximum=1), Passthrough Float=2)
    Sprite Rotation Mode=::NewEnumerator2
    Sprite Size=&MaskVector2dBySpawnGroup(
     Masked Vector2D=&RandomRangeVector2D(Minimum=<175.1472 188.2652>, Maximum=<240.9544 227.5624>)
     Passthrough Vector2D=&RandomRangeVector2D(Minimum=<430.1987 406.443>, Maximum=<694.0313 615.6079>)
     Spawn Group Mask=0
    )
    Sprite Size Mode=::NewEnumerator4
    Sprite UV Mode=::NewEnumerator3
    UsePositionOffset=false
    Write Material Random=true
   )
   &CameraOffset(Camera Offset Amount=50)
   &ShapeLocation(
    Handle Radius=5
    Large Radius=99.7365
    MaskSpawnGroup=true
    Rotation Mode=::NewEnumerator1
    Shape Primitive=::NewEnumerator3
    Torus Distribution Mode:ENiagaraTorusDistributionMode=::NewEnumerator1
    U Position=&ReturnNormalizedExecIndex()
    UseOwnerScale=true
   )
   &ShapeLocation(MaskSpawnGroup=true, Override Seed=false, Spawn Group Mask=1, Sphere Radius=25, UseOwnerScale=true) {module=ShapeLocation001}
   &AddVelocity(
    Module.Use Falloff Distance=false
    Override Seed=false
    Velocity Mode=::NewEnumerator1
    Velocity Speed=&RandomRangeFloat(Minimum=354.9144, Maximum=843.9396)
   )
  ]
  particleUpdateScript [
   &Drag(Drag=&RandomRangeFloat(Minimum=6, Maximum=11), Use Linear Drag=true, Use Rotational Drag=true)
   &SolveForcesAndVelocity(Clamp Velocity=false, Limit Acceleration=false)
   &ParticleState()
   &ScaleSpriteSize(
    Uniform Curve Scale=$User.Uniform Impact Scale
    Uniform Curve Sprite Scale=DataInterface() {propertyValues {Curve {keys [
        {arriveTangent=4.1286, interpMode=RCIM_Cubic, leaveTangent=4.1286, tangentMode=RCTM_User, time=0, value=0.475}
        {arriveTangent=1.2171, interpMode=RCIM_Cubic, leaveTangent=1.2171, tangentMode=RCTM_User, time=0.3, value=1.125}
        <at 1 1.4 RCIM_Cubic>
       ]}}}
   )
   /SetVariables(Particles.SubImageIndex=&FloatFromCurve(
     FloatCurve=DataInterface() {propertyValues {Curve {keys [
         {arriveTangent=2.5597, interpMode=RCIM_Cubic, leaveTangent=2.5597, tangentMode=RCTM_User, time=0, value=0.2}
         {arriveTangent=0.8, interpMode=RCIM_Cubic, leaveTangent=0.8, time=0.1, value=0.75}
         {interpMode=RCIM_Cubic, leaveTangent=1.1425, time=1, value=1}
        ]}}}
     Scale Curve=63
    )) {isSetParameters=true}
   &DynamicMaterialParameters(
    Index 0 Param 1=&FloatFromCurve(FloatCurve=DataInterface() {propertyValues {Curve {keys [{arriveTangent=3.9017, interpMode=RCIM_Cubic, leaveTangent=3.9017, tangentMode=RCTM_User, time=0, value=0.4}, <at 0.5 1 RCIM_Cubic>]}}})
    Index 0 Param 4=1
    Module.Index0Float0Write=true
    Module.Index0Float1Write=true
    Module.Index0Float2Write=true
    Module.Index0Float3Write=true
   )
   &ScaleColor(
    Scale Alpha=&FloatFromCurve(
     FloatCurve=DataInterface() {propertyValues {Curve {
        defaultValue=1
        keys [
         {arriveTangent=9.0725, interpMode=RCIM_Cubic, leaveTangent=9.0725, tangentMode=RCTM_User, time=0, value=0}
         {arriveTangent=-2.5249, interpMode=RCIM_Cubic, leaveTangent=-2.5249, tangentMode=RCTM_User, time=0.0441, value=0.975}
         {arriveTangent=-0.0405, interpMode=RCIM_Cubic, leaveTangent=-0.0405, tangentMode=RCTM_User, time=1, value=0}
        ]
       }}}
     Scale Curve=0.5
    )
    Scale RGB=&MakeVectorFromLinearColorRGB(Color=&ColorFromCurve(ColorCurve=DataInterface() {propertyValues {
        AlphaCurve {keys []}
        BlueCurve {keys [{arriveTangent=-9.5363, interpMode=RCIM_Cubic, leaveTangent=-9.5363, tangentMode=RCTM_User, time=0, value=4}, <at 1 1 RCIM_Cubic>]}
        GreenCurve {keys [{arriveTangent=-9.5363, interpMode=RCIM_Cubic, leaveTangent=-9.5363, tangentMode=RCTM_User, time=0, value=4}, <at 1 1 RCIM_Cubic>]}
        RedCurve {keys [{arriveTangent=-9.5363, interpMode=RCIM_Cubic, leaveTangent=-9.5363, tangentMode=RCTM_User, time=0, value=4}, <at 1 1 RCIM_Cubic>]}
       }}))
    ScaleA=true
    ScaleRGB=true
   )
   &FN_PerParticleLOD(
    Alpha Scale Near=0.5
    Color Linear Distance Power=0.5
    Enable Color Distance Scaling=true
    Enable SpriteSize Distance Scaling=true
    RGB Scale Near:Vector3f=<0.2 0.2 0.2>
    Sprite Size Scale Near:Vector2f=<0.5 0.5>
   )
  ]
  rendererClasses [&NiagaraSpriteRendererProperties]
  renderers [{rendererClass=&NiagaraSpriteRendererProperties}]
 }
 Sonic_Boom_LW {rendererClasses [], renderers []}
 Sparks {
  simTarget=GPUComputeSim
  emitterUpdateScript [
   &EmitterState(Life Cycle Mode=::NewEnumerator1, Loop Behavior=::NewEnumerator2, Loop Count=3, Loop Duration=0.1, UseLoopDelay=true)
   &SpawnRate(SpawnRate=999, Use Spawn Probability=false) {enabled=false}
   &SpawnBurst_Instantaneous(Spawn Count=3000, Use Spawn Probability=false, UseLoopCountLimit=false)
  ]
  particleSpawnScript [
   &InitializeParticle(
    Color Maximum=<rgba 0.3385 0.3385 0.3385 1>
    Color Minimum=<rgba 0.0677 0.0677 0.0677 1>
    Color Mode=::NewEnumerator2
    Lifetime=&RandomRangeFloat(Minimum=1, Maximum=3)
    Sprite Size Mode=::NewEnumerator3
    Uniform Sprite Size=&FloatFromCurve(
     FloatCurve=DataInterface() {propertyValues {Curve {keys [{arriveTangent=-2.2424, interpMode=RCIM_Cubic, leaveTangent=-2.2424, tangentMode=RCTM_User, time=0, value=1.5}, <at 1 0.85 RCIM_Cubic>]}}}
     CurveIndex=&RandomRangeFloat(Minimum=0, Maximum=1)
     Scale Curve=&Multiply_Float(A=0.7, B=$User.Uniform Impact Scale)
    )
    UsePositionOffset=false
    Write Material Random=true
   )
   &ShapeLocation(
    Cylinder Height=0
    Cylinder Height Midpoint=0
    Cylinder Radius=&UniformAOrBFloat(A=100, B=132.6338, Distribution Weight=0.412, Evaluation Type=::NewEnumerator0, Recalculate Random Each Loop=true, Override Seed=false)
    MaskSpawnGroup=false
    Override Seed=false
    Rotation Mode=::NewEnumerator2
    Shape Primitive=::NewEnumerator1
    Surface Only=true
    Use Endcaps In Surface Only Mode=false
    UseOwnerScale=true
    Yaw / Pitch / Roll=&MakeCustomVectorFromBool(
     Bool=&SetBoolByFloatComparison(A=&RandomRangeFloat(Minimum=0, Maximum=1), B=0.95)
     True Vector=&RandomVector(Vector Scale=1)
     False Vector=&SelectVectorFromArray(
      Array Sampling Mode=::NewEnumerator0
      Direct Array Index=?
      Vector Selection Array=DataInterface() {propertyValues {FloatData [<0.2252 0.236 0.2072>, <0.25 0.1351 0.3081>, <0.1081 0.2 0.3766>, <0 0.4595 0>]}}
      Random Seed=0
     )
    )
   )
   &AddVelocity(
    Module.Use Falloff Distance=false
    Override Seed=false
    Velocity Mode=::NewEnumerator1
    Velocity Speed=&RandomRangeFloat(
     Minimum=&Multiply_Float(A=1200, B=$User.Uniform Impact Scale)
     Maximum=&Multiply_Float(A=1600, B=$User.Uniform Impact Scale)
    )
   )
  ]
  particleUpdateScript [
   &ParticleState()
   &CurlNoiseForce(
    Bias Noise=false
    Noise Frequency=5
    Noise Strength=&FloatFromCurve(
     FloatCurve=DataInterface() {propertyValues {Curve {
        defaultValue=1
        keys [
         {arriveTangent=-3.2557, interpMode=RCIM_Cubic, leaveTangent=-3.2557, tangentMode=RCTM_User, time=0, value=1}
         {arriveTangent=-0.7893, interpMode=RCIM_Cubic, leaveTangent=-0.7893, tangentMode=RCTM_User, time=0.1, value=0.25}
         {interpMode=RCIM_Cubic, leaveTangent=-8.0089, time=1, value=0}
        ]
       }}}
     Scale Curve=&Multiply_Float(A=6000, B=$User.Uniform Impact Scale)
    )
    Pan Noise=false
   )
   &Drag(Drag=5, Use Linear Drag=true, Use Rotational Drag=true)
   &PointForce(Force Strength=-200, Module.Use Falloff=false, Module.UseLinearForce=false)
   &SolveForcesAndVelocity(Clamp Velocity=false, Limit Acceleration=false)
   &ScaleSpriteSizeBySpeed(Max Scale Factor=<1 50>, Velocity Threshold=2000)
   &ScaleColor(
    Scale Alpha=&FloatFromCurve(FloatCurve=DataInterface() {propertyValues {Curve {keys [<at 0 0>,<at 0.1 1>,<at 1 0>]}}}, Scale Curve=0.5)
    ScaleA=true
    ScaleRGB=true
   )
  ]
  rendererClasses [&NiagaraSpriteRendererProperties]
  renderers [{rendererClass=&NiagaraSpriteRendererProperties}]
 }
 Sphere {
  emitterUpdateScript [
   &EmitterState()
   &SpawnBurst_Instantaneous(Use Spawn Probability=false, UseLoopCountLimit=false)
  ]
  particleSpawnScript [&InitializeParticle(
    Color Mode=::NewEnumerator0
    Lifetime=0.4
    Mesh Scale Mode=::NewEnumerator3
    Mesh Uniform Scale=2
    UsePositionOffset=false
    Write Material Random=true
   )]
  particleUpdateScript [
   &ParticleState()
   &ScaleMeshSize(Scale Factor=&VectorFromFloat(Value=&FloatFromCurve(
      FloatCurve=DataInterface() {propertyValues {Curve {
         defaultValue=0
         keys [
          {arriveTangent=4.1785, interpMode=RCIM_Cubic, leaveTangent=4.1785, tangentMode=RCTM_User, time=0.1, value=0.5}
          {arriveTangent=1.1179, interpMode=RCIM_Cubic, leaveTangent=1.1179, tangentMode=RCTM_User, time=0.2, value=0.75}
          {arriveTangent=0.0224, interpMode=RCIM_Cubic, leaveTangent=0.0224, tangentMode=RCTM_User, time=1, value=1}
         ]
        }}}
      Scale Curve=$User.Uniform Impact Scale
     )))
   &DynamicMaterialParameters(
    Index 0 Param 4=&FloatFromCurve(FloatCurve=DataInterface() {propertyValues {Curve {keys [<at 0 1 RCIM_Cubic>,<at 1 0 RCIM_Cubic>]}}}, Scale Curve=0.05)
    Module.Index0Float0Write=true
    Module.Index0Float1Write=true
    Module.Index0Float2Write=true
    Module.Index0Float3Write=true
   )
   &ScaleColor(
    Scale Alpha=&FloatFromCurve(FloatCurve=DataInterface() {propertyValues {Curve {keys [<at 0 1>,<at 1 0>]}}})
    ScaleA=true
    ScaleRGB=true
   )
   &FN_PerParticleLOD(Alpha Scale Near=0.1, Color Linear Distance Power=0.5, Enable Color Distance Scaling=true, RGB Scale Near:Vector3f=<0.2 0.2 0.2>)
  ]
  rendererClasses [&NiagaraMeshRendererProperties]
  renderers [{rendererClass=&NiagaraMeshRendererProperties, rendererIndex=0}]
 }
 Sphere001 {
  emitterUpdateScript [
   &EmitterState()
   &SpawnBurst_Instantaneous(Use Spawn Probability=false, UseLoopCountLimit=false)
  ]
  particleSpawnScript [&InitializeParticle(
    Color Mode=::NewEnumerator0
    Lifetime=2
    Mesh Scale Mode=::NewEnumerator3
    Mesh Uniform Scale=4
    UsePositionOffset=false
    Write Material Random=true
   )]
  particleUpdateScript [
   &ParticleState()
   &ScaleMeshSize(Scale Factor=&VectorFromFloat(Value=&FloatFromCurve(
      FloatCurve=DataInterface() {propertyValues {Curve {
         defaultValue=0
         keys [
          {arriveTangent=4.1785, interpMode=RCIM_Cubic, leaveTangent=4.1785, tangentMode=RCTM_User, time=0, value=0}
          {arriveTangent=1.1179, interpMode=RCIM_Cubic, leaveTangent=1.1179, tangentMode=RCTM_User, time=0.2042, value=0.6228}
          {arriveTangent=0.0224, interpMode=RCIM_Cubic, leaveTangent=0.0224, tangentMode=RCTM_User, time=1, value=1}
         ]
        }}}
      Scale Curve=$User.Uniform Impact Scale
     )))
   &DynamicMaterialParameters(
    Index 0 Param 4=&FloatFromCurve(FloatCurve=DataInterface() {propertyValues {Curve {keys [<at 0 1 RCIM_Cubic>,<at 1 0 RCIM_Cubic>]}}}, Scale Curve=0.02)
    Module.Index0Float0Write=true
    Module.Index0Float1Write=true
    Module.Index0Float2Write=true
    Module.Index0Float3Write=true
   )
   &ScaleColor(
    Scale Alpha=&FloatFromCurve(FloatCurve=DataInterface() {propertyValues {Curve {keys [<at 0 1>,<at 1 0>]}}}, Scale Curve=0.2)
    ScaleA=true
    ScaleRGB=true
   )
  ]
  rendererClasses [&NiagaraMeshRendererProperties]
  renderers [{rendererClass=&NiagaraMeshRendererProperties, rendererIndex=0}]
 }
 WindCone {
  emitterUpdateScript [
   &EmitterState(Inactive Response=::NewEnumerator1, Life Cycle Mode=::NewEnumerator1, Loop Behavior=::NewEnumerator1, Loop Duration=1, UseLoopDelay=true)
   &SpawnBurst_Instantaneous(Use Spawn Probability=false, UseLoopCountLimit=false)
  ]
  particleSpawnScript [
   &InitializeParticle(Color Mode=::NewEnumerator0, Lifetime=0.3, Mesh Scale Mode=::NewEnumerator4, Position Mode=::NewEnumerator0, Write Material Random=true)
   &InitialMeshOrientation(Mesh Orientation Mode=::NewEnumerator2, Rotation=<0 0.25 0>, Use Rotation Vector=true)
  ]
  particleUpdateScript [
   &ParticleState()
   &UpdateMeshOrientation(
    Facing Direction=&NormalizeVector(Vector To Normalize=$Engine.Owner.Velocity, Safe Normalize=true, Safe Fallback Threshold=0, Safe Fallback Vector=<0 0 0.001>)
    Facing Interpolation=false
    Orientation Method=::NewEnumerator1
    Override Quat=false
    TransformFacing=false
    TransformUp=false
   ) {enabled=false}
   &ScaleMeshSize(Scale Factor=&VectorFromFloat(Value=&FloatFromCurve(
      FloatCurve=DataInterface() {propertyValues {Curve {defaultValue=0, keys [{tangentMode=RCTM_User, time=0, value=0.5}, {tangentMode=RCTM_User, time=1, value=1}]}}}
      Scale Curve=&Multiply_Float(A=0.8, B=$User.Uniform Impact Scale)
     )))
   &ScaleColor(
    Scale Alpha=&FloatFromCurve(
     FloatCurve=DataInterface() {propertyValues {Curve {keys [{arriveTangent=-1.2369, interpMode=RCIM_Cubic, leaveTangent=-1.2369, tangentMode=RCTM_User, time=0.1, value=0.8}, <at 1 0 RCIM_Cubic>]}}}
     Scale Curve=2
    )
    ScaleA=true
    ScaleRGB=true
   )
   &DynamicMaterialParameters(
    Index 0 Param 1=0.5
    Index 0 Param 2=10
    Module.Index0Param0Write=true
    Module.Index0Param1Write=true
    Module.Index0Param2Write=true
    Module.Index0Param3Write=true
   )
  ]
  rendererClasses [&NiagaraMeshRendererProperties]
  renderers [{rendererClass=&NiagaraMeshRendererProperties, rendererIndex=0}]
 }
}
summary {
 systemName=NS_Pistol_Canary_WealthLamb_Punch
 emitters [
  {emitterName=Smoke, rendererClasses [&NiagaraSpriteRendererProperties]}
  {emitterName=Sphere, rendererClasses [&NiagaraMeshRendererProperties]}
  {emitterName=Sphere001, rendererClasses [&NiagaraMeshRendererProperties]}
  {emitterName=Sparks, simTarget=GPUComputeSim, rendererClasses [&NiagaraSpriteRendererProperties]}
  {emitterName=WindCone, rendererClasses [&NiagaraMeshRendererProperties]}
  {emitterName=Sonic_Boom_LW, rendererClasses []}
  {emitterName=ShockWave, rendererClasses [&NiagaraSpriteRendererProperties]}
 ]
 userVariables [
  User.Uniform Curve Scale {type=NiagaraFloat, defaultValue=0, description=""}
  User.Uniform Impact Scale {type=NiagaraFloat, defaultValue=5, description=""}
 ]
}
